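; Weapon mod bench actor that players interact with to mod weapons, destroyed by the server where the mode disallows it.
; Interaction checks range, facing and interacting tag, then starts the session on the controller's WeaponModStationPlayerComponent and grants a camera ability.
; Used effects activate upgrade FX and a timeline-driven mesh wobble.
; When linked to a vault door, it hides its special actor component, used for the minimap icon, until a matching byte event arrives.
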(event Custom|TriggerStationUsedEffects
  (bind _p_upgradefx (Variables|Default|GetP_UpgradeFX))
  (Utilities|FlowControl|Sequence
    (:then_0
      (Utilities|IsValid _p_upgradefx
        (:Is%20Valid
          (Components|Activation|SetActive _p_upgradefx true true))))
    (:then_1
      (bind (_direction _wobbleovertime) (Timeline|WobbleTimeline PlayFromStart)
        (:Update
          (Transformation|SetRelativeRotation (Variables|Debug|GetStaticMeshComponent) (Math|Rotator|CombineRotators (Variables|Default|GetInitialRelativeRotation) (Math|Rotator|Lerp(Rotator) "0, 0, 0" "0.500000,0.000000,0.000000" _wobbleovertime true)) false false))))))

(event Collision|EventActorBeginOverlap
  (params (OtherActor :type Actor%20Object%20Reference))
  (|CallParentFunction OtherActor) ; This node is disabled and will not be called.\nDrag off pins to build functionality.
)

(event EventTick
  (params (DeltaSeconds :type Float%20%28single-precision%29))
  (|CallParentFunction DeltaSeconds) ; This node is disabled and will not be called.\nDrag off pins to build functionality.
)

(event EventEndPlay
  (params (EndPlayReason :type EEndPlayReason%20Enum :default "Destroyed"))
  (|CallParentFunction EndPlayReason))

(event EventBeginPlay
  (|CallParentFunction)
  (Fortnite|SetCanbeDamaged self (ScalableFloat|GetValueasBool (Variables|Hotfix|GetCanBeDamaged_SF)))
  (Sound|SetUpSound)
  (Variables|Default|SetInitialRelativeRotation (Variables|/Script/Engine.SceneComponent|GetRelativeRotation (Variables|Debug|GetStaticMeshComponent)))
  ;[0.281 1 0.4092] Destroy the mod bench if not supported in Creative Royale.
  (Networking|SwitchHasAuthority
    (:Authority
      (if (not (ScalableFloat|GetValueasBool (Variables|Hotfix|GetAllowedInMode)))
        (Actor|DestroyActor)))))

(event Interaction|EventBlueprintOnInteract
  (params (InteractingPawn :type Fort%20Pawn%20Object%20Reference) (InteractionBeingAttempted :type EInteractionBeingAttempted%20Enum :default "FirstInteraction"))
  ;[1 1 1] [SERVER]
  (Utilities|FlowControl|Sequence
    (:then_0
      (bind _asfort_player_controller (Utilities|Casting|CastToFortPlayerController (Pawn|GetController InteractingPawn)))
      (Utilities|IsValid (Actor|GetComponentbyClass _asfort_player_controller "/WeaponModStation/Components/WeaponModStationPlayerComponent.WeaponModStationPlayerComponent_C")
        (:Is%20Valid
          (Class|FortWeaponModStationComponent|StartInteractwithWorkbenchActor (Actor|GetComponentbyClass _asfort_player_controller "/WeaponModStation/Components/WeaponModStationPlayerComponent.WeaponModStationPlayerComponent_C") self)
          (EventDispatchers|BindEventtoOnStartInteractwithWorkbenchActor (Actor|GetComponentbyClass _asfort_player_controller "/WeaponModStation/Components/WeaponModStationPlayerComponent.WeaponModStationPlayerComponent_C") (EventDispatchers|CreateEvent :!event Custom|On%20Start%20Interact%20Event))
          (EventDispatchers|BindEventtoOnStopInteractwithWorkbenchActor (Actor|GetComponentbyClass _asfort_player_controller "/WeaponModStation/Components/WeaponModStationPlayerComponent.WeaponModStationPlayerComponent_C") (EventDispatchers|CreateEvent :!event Custom|On%20Stop%20Interact%20Event)))))
    (:then_1
      ;;[0 0.07082 1] Apply GA to Push new camera while the player is interacting with the mod system. The GA handles removing itself when the WeaponModStation.LocalInteracting tag is removed
      ;;
      (Ability|Advanced|GrantandActivateAbilityonSelf (Ability|GetAbilitySystemComponent InteractingPawn) "/WeaponModStation/Gameplay/Abilities/GA_WeaponModStation_Camera_v2.GA_WeaponModStation_Camera_v2_C")
      (Ability|Tasks|WaitGameplayTagRemoveFromActor 0 "(TagName=\"WeaponModStation.LocalInteracting\")" true))))

(event Custom|On%20Start%20Interact%20Event
  (params (NewInteractingWeaponModStation :type Fort%20Weapon%20Mod%20Station%20Base%20Object%20Reference)))

(event Custom|On%20Stop%20Interact%20Event
  (params (WeaponModStation :type Fort%20Weapon%20Mod%20Station%20Base%20Object%20Reference)))

(event Custom|ListenForVaultOpenEvent
  (params (Byte :type Byte))
  ;[0 0.1227 1] If we are supporting a linked vault door to determine minimap icon visibility, set it here.
  ; Called from B_Athena_Spawner_WeaponModStation_V2
  ; Spawner does a valid check, so if we get here, start the deactivate icon process and wait for vault opened
  ; Event Tags are set per instance in the map using tags:
  ; Event.VaultDoorModBench.Opened.One through Opened.Five
  (Variables|Default|SetVault%20Opened%20Event%20Byte Byte)
  (bind _returnvalue (GameplayEventRouter|GetEventRouter "/Script/FortniteGame.FortPlayspace" self))
  (Utilities|IsValid _returnvalue
    (:Is%20Valid
      (Components|Activation|Deactivate (Variables|Default|GetWeaponModStationSpecialActorComponent))
      (bind (_asynctaskproxy _payload) (GameplayEventRouter|StartListeningToEvent _returnvalue "/Game/Athena/BroadcastEventStructs/Struct_BroadCast_Byte_Event.Struct_BroadCast_Byte_Event")
        (:OnEventReceived
          (if (== (Utilities|Struct|BreakStructBroadCastByteEvent _payload) (Variables|Default|GetVault%20Opened%20Event%20Byte))
            (Components|Activation|Activate (Variables|Default|GetWeaponModStationSpecialActorComponent) false)))))))

(fn BlueprintCanInteract
  (params (InteractingPawn :type Fort%20Pawn%20Object%20Reference) (InteractionBeingAttempted :type EInteractionBeingAttempted%20Enum :default "FirstInteraction") (InteractionType :type TInteractionType%20Enum :default "IT_NoInteraction"))
  (return (Math|Boolean|ANDBoolean (< (Math|Vector|Distance2DSquared(Vector) (Transformation|GetActorLocation InteractingPawn) (Transformation|GetWorldLocation (Variables|Default|GetTableCollision))) (LocalVariables|GetInteractDist2DSq)) (not (GameplayTags|HasMatchingGameplayTag (Ability|GetAbilitySystemComponent InteractingPawn) "(TagName=\"WeaponModStation.LocalInteracting\")")) (> (Math|Vector|DotProduct (Math|Vector|Normalize (- (Transformation|GetActorLocation InteractingPawn) (Transformation|GetWorldLocation (Variables|Default|GetTableCollision)))) (Math|Vector|Normalize (- (Utilities|Operators|Add (Transformation|GetWorldLocation (Variables|Default|GetTableCollision)) (* (Math|Vector|GetRightVector (Transformation|GetWorldRotation (Variables|Default|GetTableCollision))) 1e+02) "0.000000,0.000000,50.000000") (Transformation|GetWorldLocation (Variables|Default|GetTableCollision))))) 0.45))))

(fn BlueprintGetFailedInteractionString
  (params (InteractingPawn :type Fort%20Pawn%20Object%20Reference) (InteractionBeingAttempted :type EInteractionBeingAttempted%20Enum :default "FirstInteraction"))
  (return ""))

(fn BlueprintGetInteractionString
  (params (InteractingPawn :type Fort%20Pawn%20Object%20Reference) (InteractionBeingAttempted :type EInteractionBeingAttempted%20Enum :default "FirstInteraction"))
  (return (Variables|Default|GetInteractString)))

(fn BlueprintGetInteractionTime
  (params (InteractingPawn :type Fort%20Pawn%20Object%20Reference) (InteractionBeingAttempted :type EInteractionBeingAttempted%20Enum :default "FirstInteraction"))
  (return 0.0 true))

(fn PlayServiceUsedEffects
  (params (PlayerController :type Fort%20Player%20Controller%20Object%20Reference))
  (CallFunction|TriggerStationUsedEffects))

(fn SetUpSound
  (bind _weaponmodstation_ambient (Variables|Default|GetWeaponModStation_Ambient))
  (Utilities|IsValid _weaponmodstation_ambient
    (:Is%20Valid
      (Audio|Components|Audio|Play _weaponmodstation_ambient 0.0))))

(fn ConstructionScript
  (|CallParentFunction))
