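; Spawner for the Weapon Mod Station v2 that configures each spawned station once playlist data is ready, if the base hotfix is enabled.
; Stations at secondary locations get a short lifespan when the secondary-location hotfix is off, culling them.
; On authority, stations either listen for a vault-opened event to drive minimap icon visibility or have their special actor component deactivated to hide it by default, then receive configured loose gameplay tags.
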(event EventServerActorSpawned
  (params (Actor :type Actor%20Object%20Reference))
  ;[0 0.947 1] We may limit total spawn locations not based on spawn chance, but location. If spawner is marked as secondary w/ bool, then check hotfix to see if we're spawning
  (Class|FortAsyncActionCurrentPlaylistReady|PlaylistDataReadyAsync
    (:OnReady
      (if (ScalableFloat|GetValueasBool (Variables|Default|GetIsBaseEnabledHotfix))
        (if (Variables|Default|GetIsSecondaryLocationSpawner)
          (if (not (ScalableFloat|GetValueasBool (Variables|Default|GetSecondarySpawnLocationsEnabledHotfix)))
            (Actor|SetLifeSpan Actor 1.0))
          (else
            _))
        ;[0 0.1164 1] If we are supporting a linked vault door to determine minimap icon visibility, send it to the spawned actor
        (Networking|SwitchHasAuthority
          (:Authority
            (Utilities|FlowControl|Sequence
              (:then_0
                (if (and (ScalableFloat|GetValueasBool (Variables|Default|GetOnlyShowMapIconWhenVaultOpened)) (!= (Variables|Default|GetVault%20Opened%20Event%20Byte) 0))
                  (bind _asbp_weapon_mod_station_v_2 (Utilities|Casting|CastToBP_WeaponModStation_v2 Actor))
                  (Class|BPWeaponModStationV2|ListenforVaultOpenEvent _asbp_weapon_mod_station_v_2 (Variables|Default|GetVault%20Opened%20Event%20Byte))
                  (elif (Variables|Default|GetbDisableMapIconByDefault)
                    (bind _asbp_weapon_mod_station_v_2_2 (Utilities|Casting|CastToBP_WeaponModStation_v2 Actor))
                    (bind _weaponmodstationspecialactorcomponent (Variables|/WeaponModStation/Gameplay/Actors/BP_WeaponModStation_v2.BP_WeaponModStation_v2_C|GetWeaponModStationSpecialActorComponent _asbp_weapon_mod_station_v_2_2)
                      (:then
                        (Components|Activation|Deactivate _weaponmodstationspecialactorcomponent))
                      (:else)))))
              (:then_1
                ;[0.08849 0.2987 0.3177] Apply loose gameplay tags.
                (bind _asbp_weapon_mod_station_v_2_3 (Utilities|Casting|CastToBP_WeaponModStation_v2 Actor))
                (Ability|GameplayEffect|AddLooseGameplayTags _asbp_weapon_mod_station_v_2_3 (Variables|Default|GetLooseGameplayTagsToApply) true)))))))))

(event EventBeginPlay
  (|CallParentFunction) ; This node is disabled and will not be called.\nDrag off pins to build functionality.
)

(event Collision|EventActorBeginOverlap
  (params (OtherActor :type Actor%20Object%20Reference)))

(event EventTick
  (params (DeltaSeconds :type Float%20%28single-precision%29)))

(fn ConstructionScript
  (|CallParentFunction))
